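# Niagara system for the Dark Matter vehicle boost cosmetic, trailing black and magenta exhaust from each boost socket.
# Filtered socket count drives ribbon count and particle partitioning across skeletal mesh sockets.
# Two active ribbon emitters (a third is disabled) layer with curl-noise sprites, short-lived mesh/sprite bursts split by spawn group, and persistent socket mesh/sprite glows.
# Painted-variant modules and user paint parameters recolor the magenta layers; ribbon dark cores stay black.
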NiagaraSystem /VehicleCosmetics/Boosts/Boost_DarkMatter/NS_Boost_DarkMatter
dependencies {
 usedDataInterfaces [
  &NiagaraDataInterfaceCurve
  &NiagaraDataInterfaceArrayFloat
  &NiagaraDataInterfaceArrayFloat3
  &NiagaraDataInterfaceVectorCurve
  &NiagaraDataInterfaceSkeletalMesh
 ]
 usedDynamicInputs []
 usedModules [
  &self:SystemSpawnScript.SystemScriptSource.SystemScriptGraph.NiagaraNodeAssignment_0.SetVariables_#1
  &SystemState
 ]
}
emitters {
 NE_VC_Boost_Ribbon_Parent {
  bEnabled=false
  emitterUpdateScript [
   &EmitterState()
   &VC_FilteredSocketCount()
   &VC_RateQualityHandler()
   &VC_SpawnRibbon_v1(Number of Ribbons=$Emitter.PortCount, SpawnRate=$Emitter.Spawn Rate)
  ]
  particleSpawnScript [
   /VehicleCosmetics/FX/Niagara/Modules/MergeEmitters/VC_ExtraBindingColors(
    ExtraColors=4
    R1 Color:LinearColor=<rgba 0 0 0 1>
    R2 Color:LinearColor=&MakeCustomLinearColorFromBool(Bool=$User.bIsPainted, True Color=$User.PaintedColor, False Color=<rgba 8 1 8 1>)
    R3 Color:LinearColor=&MakeCustomLinearColorFromBool(Bool=$User.bIsPainted, True Color=$User.PaintedColor, False Color=<rgba 36 4.5 35.9999 1>)
    R4 Color:LinearColor=&MakeCustomLinearColorFromBool(Bool=$User.bIsPainted, True Color=$User.PaintedColor, False Color=<rgba 0.6 0.075 0.6 1>)
   )
   &VC_PaintedVariantsOverride(
    R2 Black:LinearColor=<rgba 0.112 0.112 0.112 1>
    R2 BurntSienna:LinearColor=<rgba 1.25 0.2917 0 1>
    R2 Cobalt:LinearColor=<rgba 0.12 0.36 3 1>
    R2 Crimson:LinearColor=<rgba 1.35 0 0.0366 1>
    R2 DefaultColor:LinearColor=<rgba 8 1 8 1>
    R2 ForestGreen:LinearColor=<rgba 0 1.35 0 1>
    R2 Lime:LinearColor=<rgba 1 2 0 1>
    R2 Orange:LinearColor=<rgba 3 0.55 0 1>
    R2 Pink:LinearColor=<rgba 0.5052 0.0605 0.3435 1>
    R2 Purple:LinearColor=<rgba 0.6667 0 2 1>
    R2 Saffron:LinearColor=<rgba 3 2.3 0 1>
    R2 SkyBlue:LinearColor=<rgba 0 1.3308 2 1>
    R2 TitaniumWhite:LinearColor=<rgba 2 2 2 1>
    Use R2 Painted Overrides=true
   )
   /SetVariables(Particles.DarkRibbon1Size=80, Particles.DarkRibbon2Size=47.5) {isSetParameters=true, module=SetVariables_#2}
   &VC_InitializeRibbon_v1(Custom Offset=true, Enable Skeletal Position=true, Lifetime=0.8, Local Vector Offset=<-20 0 0>, Offset Amount=1, Ribbon Width=40)
   &AddVelocity(Velocity=<-2000 0 0>) {module=AddVelocity001}
  ]
  particleUpdateScript [
   &ParticleState()
   &ScaleRibbonWidth(Ribbon Width Scale=&FloatFromCurve(FloatCurve=DataInterface() {propertyValues {Curve {keys [<at 0 1>,<at 0.75 0>]}}}))
   /SetVariables(Particles.GlowRibbonScale:NiagaraFloat=&Multiply_Float(
     A=$Particles.Initial.RibbonWidth
     B=&FloatFromCurve(FloatCurve=DataInterface() {propertyValues {Curve {keys [<at 0 1>,<at 0.75 0>]}}}, Scale Curve=2.35)
    )) {isSetParameters=true, module=SetVariables_#3}
   &SolveForcesAndVelocity(Clamp Velocity=false, Limit Acceleration=false)
   &VC_UpdateRibbon_v1(Enable Skeletal Position=true)
  ]
  rendererClasses [&NiagaraRibbonRendererProperties]
  renderers [
   {rendererClass=&NiagaraRibbonRendererProperties}
   {rendererClass=&NiagaraRibbonRendererProperties, rendererIndex=1}
   {rendererClass=&NiagaraRibbonRendererProperties, rendererIndex=2}
   {rendererClass=&NiagaraRibbonRendererProperties, rendererIndex=3}
  ]
 }
 NE_VC_Boost_Ribbon_Parent001 {
  emitterUpdateScript [
   &EmitterState()
   &VC_FilteredSocketCount()
   &VC_RateQualityHandler(Default Rate=30)
   &VC_SpawnRibbon_v1(Number of Ribbons=$Emitter.PortCount, SpawnRate=$Emitter.Spawn Rate)
  ]
  particleSpawnScript [
   &VC_ExtraBindingColors(
    ExtraColors=1
    R1 Color:LinearColor=&MakeCustomLinearColorFromBool(Bool=$User.bIsPainted, True Color=$User.PaintedColor, False Color=<rgba 4 0.5 4 1>)
   )
   &VC_PaintedVariantsOverride(
    R1 BurntSienna:LinearColor=<rgba 32 9.8801 3.9445 1>
    R1 Cobalt:LinearColor=<rgba 0 7.4667 64 1>
    R1 Crimson:LinearColor=<rgba 64 0 1.7346 1>
    R1 DefaultColor=<rgba 64 8 63.9998 1>
    R1 ForestGreen:LinearColor=<rgba 0 64 3.2337 1>
    R1 Lime:LinearColor=<rgba 37.3333 64 0 1>
    R1 Orange:LinearColor=<rgba 64 12.8 0 1>
    R1 Purple:LinearColor=<rgba 8.5334 0 64 1>
    R1 Saffron:LinearColor=<rgba 64 49.0667 0 1>
    R1 SkyBlue:LinearColor=<rgba 0 42.5847 64 1>
    R1 TitaniumWhite:LinearColor=<rgba 64 64 64 1>
    Use R1 Painted Overrides=true
   )
   &VC_InitializeRibbon_v1(Custom Offset=true, Enable Skeletal Position=true, Lifetime=0.8, Local Vector Offset=<-50 0 0>, Offset Amount=1, Ribbon Width=7.5)
   &AddVelocity(Velocity=<-1800 0 0>) {module=AddVelocity001}
  ]
  particleUpdateScript [
   &ParticleState()
   &SolveForcesAndVelocity(Clamp Velocity=false, Limit Acceleration=false)
   &VC_UpdateRibbon_v1(Enable Skeletal Position=true)
   &ScaleColor(
    Scale RGB=&VectorFromCurve(VectorCurve=DataInterface() {propertyValues {XCurve {keys [<at 0 0.6>,<at 0.15 0.6>]}, YCurve {keys [<at 0 0.6>,<at 0.15 0.6>]}, ZCurve {keys [<at 0 0>,<at 0.15 0.6>]}}})
    ScaleA=true
    ScaleRGB=true
   )
   &ScaleRibbonWidth(Ribbon Width Scale=&FloatFromCurve(FloatCurve=DataInterface() {propertyValues {Curve {keys [<at 0 1>,<at 1 0>]}}}))
  ]
  rendererClasses [&NiagaraRibbonRendererProperties]
  renderers [
   {rendererClass=&NiagaraRibbonRendererProperties}
   {rendererClass=&NiagaraRibbonRendererProperties, rendererIndex=1}
   {rendererClass=&NiagaraRibbonRendererProperties, rendererIndex=2}
  ]
 }
 NE_VC_Boost_Ribbon_Parent002 {
  emitterUpdateScript [
   &EmitterState()
   &VC_FilteredSocketCount()
   &VC_RateQualityHandler()
   &VC_SpawnRibbon_v1(Number of Ribbons=$Emitter.PortCount, SpawnRate=$Emitter.Spawn Rate)
  ]
  particleSpawnScript [
   &VC_ExtraBindingColors(
    ExtraColors=3
    R1 Color:LinearColor=<rgba 0 0 0 1>
    R2 Color:LinearColor=&MakeCustomLinearColorFromBool(Bool=$User.bIsPainted, True Color=$User.PaintedColor, False Color=<rgba 8 1 8 1>)
    R3 Color:LinearColor=&MakeCustomLinearColorFromBool(Bool=$User.bIsPainted, True Color=$User.PaintedColor, False Color=<rgba 36 4.5 35.9999 1>)
   )
   &VC_PaintedVariantsOverride(
    R2 Black:LinearColor=<rgba 0.3 0.3 0.3 1>
    R2 BurntSienna:LinearColor=<rgba 12 2.8 0 1>
    R2 Cobalt:LinearColor=<rgba 5.12 10.0266 64 1>
    R2 Crimson:LinearColor=<rgba 16 0 0.4337 1>
    R2 DefaultColor:LinearColor=<rgba 32 4 31.9999 1>
    R2 ForestGreen:LinearColor=<rgba 0 32 0 1>
    R2 Lime:LinearColor=<rgba 18.6667 32 0 1>
    R2 Orange:LinearColor=<rgba 32 6.4 0 1>
    R2 Pink:LinearColor=<rgba 0.5052 0.0605 0.3435 1>
    R2 Purple:LinearColor=<rgba 4.2667 0 32 1>
    R2 Saffron:LinearColor=<rgba 32 24.5333 0 1>
    R2 SkyBlue:LinearColor=<rgba 0 21.2924 32 1>
    R2 TitaniumWhite:LinearColor=<rgba 32 32 32 1>
    Use R2 Painted Overrides=true
   )
   /SetVariables(Particles.DarkRibbon1Size=80, Particles.DarkRibbon2Size=47.5) {isSetParameters=true, module=SetVariables_#2}
   &VC_InitializeRibbon_v1(Custom Offset=true, Enable Skeletal Position=true, Lifetime=0.8, Local Vector Offset=<-35 0 0>, Offset Amount=1, Ribbon Width=40)
   &AddVelocity(Velocity=<-2000 0 0>) {module=AddVelocity001}
  ]
  particleUpdateScript [
   &ParticleState()
   &ScaleRibbonWidth(Ribbon Width Scale=&FloatFromCurve(FloatCurve=DataInterface() {propertyValues {Curve {keys [<at 0 1>,<at 0.75 0>]}}}))
   /SetVariables(Particles.GlowRibbonScale:NiagaraFloat=&Multiply_Float(
     A=$Particles.Initial.RibbonWidth
     B=&FloatFromCurve(FloatCurve=DataInterface() {propertyValues {Curve {keys [<at 0 1>,<at 0.75 0>]}}}, Scale Curve=2.35)
    )) {isSetParameters=true, module=SetVariables_#3}
   &InheritSourceMovement(
    Applied Position Scale=&Multiply_VectorByFloat(
     Vector=<1 1 1>
     Float=&FloatFromCurve(FloatCurve=DataInterface() {propertyValues {Curve {keys [
          {arriveTangent=-0.3067, interpMode=RCIM_Cubic, time=0, value=1}
          {arriveTangent=-2.3257, interpMode=RCIM_Cubic, leaveTangent=-2.3257, tangentMode=RCTM_User, time=1, value=0}
         ]}}})
    )
    ApplyForce=false
    ApplyPosition=true
    ApplyPositionFalloff=false
   )
   &SolveForcesAndVelocity(Clamp Velocity=false, Limit Acceleration=false)
   &VC_UpdateRibbon_v1(Enable Skeletal Position=true)
  ]
  rendererClasses [&NiagaraRibbonRendererProperties]
  renderers [
   {rendererClass=&NiagaraRibbonRendererProperties}
   {rendererClass=&NiagaraRibbonRendererProperties, rendererIndex=1}
   {rendererClass=&NiagaraRibbonRendererProperties, rendererIndex=2}
   {rendererClass=&NiagaraRibbonRendererProperties, rendererIndex=3}
  ]
 }
 NE_VC_Boost_Sprite_Parent {
  emitterUpdateScript [
   &EmitterState()
   &VC_FilteredSocketCount()
   &SpawnRate(SpawnRate=40, Use Spawn Probability=false)
  ]
  particleSpawnScript [
   &VC_ExtraBindingColors(
    ExtraColors=2
    R1 Color:LinearColor=&MakeCustomLinearColorFromBool(Bool=$User.bIsPainted, True Color=$User.PaintedColor, False Color=<rgba 5 0.625 5 1>)
    R2 Color:LinearColor=<rgba 0.01 0.01 0.01 1>
   )
   &VC_PaintedVariantsOverride(
    R1 Black:LinearColor=<rgba 0.112 0.112 0.112 1>
    R1 BurntSienna:LinearColor=<rgba 4 0.9333 0 1>
    R1 Cobalt:LinearColor=<rgba 1.28 4.864 32 1>
    R1 Crimson:LinearColor=<rgba 6 0 0.1626 1>
    R1 DefaultColor=<rgba 20 2.5 19.9999 1>
    R1 ForestGreen:LinearColor=<rgba 0 16 0 1>
    R1 Lime:LinearColor=<rgba 9.3333 16 0 1>
    R1 Orange:LinearColor=<rgba 20 4 0 1>
    R1 Purple:LinearColor=<rgba 2.1333 0 16 1>
    R1 Saffron:LinearColor=<rgba 20 15.3333 0 1>
    R1 SkyBlue:LinearColor=<rgba 0 10.6462 16 1>
    R1 TitaniumWhite:LinearColor=<rgba 20 20 20 1>
    Use R1 Painted Overrides=true
   )
   &InitializeParticle(
    Color Mode=::NewEnumerator0
    Lifetime=0.6
    Sprite Rotation Mode=::NewEnumerator2
    Sprite Size Mode=::NewEnumerator1
    Uniform Sprite Size Min=6
    UsePositionOffset=false
    Write Material Random=true
   )
   &PartitionParticles(Number Of Partitions=$Emitter.PortCount)
   &SkeletalMeshLocation(
    Bone / Socket Index=$Particles.PartitionParticles.Partition
    Bone / Socket Sampling Mode=::NewEnumerator5
    MaskSpawnGroup=false
    Mesh Sampling Type=::NewEnumerator6
    Skeletal Mesh=$System.SkeletalMesh
   )
   &AddVelocity(Velocity=<-1500 0 0>)
   /SetVariables(Particles.BlobSize:Vector2f=/Niagara/DynamicInputs/Add/Add_Vector2(A:Vector2f=$Particles.SpriteSize, B:Vector2f=&Vector2DFromFloat(Value=2.5))) {isSetParameters=true}
   &ShapeLocation(MaskSpawnGroup=false, Override Seed=false, Sphere Radius=35, UseOwnerScale=true)
  ]
  particleUpdateScript [
   &ParticleState()
   &ScaleSpriteSize(Uniform Curve Sprite Scale=DataInterface() {propertyValues {Curve {keys [{tangentMode=RCTM_User, time=0, value=1}, {tangentMode=RCTM_User, time=1, value=0}]}}})
   &CurlNoiseForce(Bias Noise=false, Noise Strength=450, Pan Noise=false)
   &SolveForcesAndVelocity(Clamp Velocity=false, Limit Acceleration=false)
   /SetVariables(
    Particles.BlobScaleSize:Vector2f=&Multiply_Vector2D_ByFloat(
     Float=&FloatFromCurve(FloatCurve=DataInterface() {propertyValues {Curve {keys [<at 0 1>,<at 1 0>]}}})
     Vector2D=$Particles.BlobSize
    )
    Particles.R1 Color:LinearColor=&MultiplyLinearColor(
     Linear Color=$Particles.Initial.R1 Color
     Scale Factor=&MakeVector4(
      X=1
      Y=1
      Z=1
      W=&FloatFromCurve(FloatCurve=DataInterface() {propertyValues {Curve {keys [<at 0.5 1>,<at 1 0>]}}})
     )
    )
   ) {isSetParameters=true}
  ]
  rendererClasses [&NiagaraSpriteRendererProperties]
  renderers [
   {rendererClass=&NiagaraSpriteRendererProperties}
   {rendererClass=&NiagaraSpriteRendererProperties, rendererIndex=1}
  ]
 }
 NE_VC_Boost_Sprite_Parent001 {
  emitterUpdateScript [
   &EmitterState()
   &VC_FilteredSocketCount()
   &SpawnRate(SpawnRate=16, Use Spawn Probability=false)
   &SpawnRate(SpawnGroup=1, SpawnRate=65, Use Spawn Probability=false) {module=SpawnRate001}
  ]
  particleSpawnScript [
   &VC_ExtraBindingColors(
    ExtraColors=2
    R1 Color:LinearColor=&MakeCustomLinearColorFromBool(Bool=$User.bIsPainted, True Color=$User.PaintedColor, False Color=<rgba 16 2 16 1>)
    R2 Color:LinearColor=&MakeCustomLinearColorFromBool(Bool=$User.bIsPainted, True Color=$User.PaintedColor, False Color=<rgba 2 0.25 2 1>)
   )
   &VC_PaintedVariantsOverride(
    R1 Black:LinearColor=<rgba 16 16 16 1>
    R1 BurntSienna:LinearColor=<rgba 64 19.7601 7.8889 1>
    R1 Cobalt:LinearColor=<rgba 10.24 23.9787 128 1>
    R1 Crimson:LinearColor=<rgba 128 0 3.4692 1>
    R1 DefaultColor=<rgba 32 4 31.9999 1>
    R1 ForestGreen:LinearColor=<rgba 0 128 6.4673 1>
    R1 Lime:LinearColor=<rgba 74.6667 128 0 1>
    R1 Orange:LinearColor=<rgba 128 25.6 0 1>
    R1 Purple:LinearColor=<rgba 17.0667 0 128 1>
    R1 Saffron:LinearColor=<rgba 128 98.1333 0 1>
    R1 SkyBlue:LinearColor=<rgba 0 85.1695 128 1>
    R1 TitaniumWhite:LinearColor=<rgba 128 128 128 1>
    R2 Black:LinearColor=<rgba 0.15 0.15 0.15 1>
    R2 BurntSienna:LinearColor=<rgba 1.75 0.5403 0.2157 1>
    R2 Cobalt:LinearColor=<rgba 0.2 0.4683 2.5 1>
    R2 Crimson:LinearColor=<rgba 1 0 0.0271 1>
    R2 DefaultColor:LinearColor=<rgba 2 0.25 2 1>
    R2 ForestGreen:LinearColor=<rgba 0 0.8 0.0404 1>
    R2 Lime:LinearColor=<rgba 0.875 1.5 0 1>
    R2 Orange:LinearColor=<rgba 1.75 0.525 0 1>
    R2 Purple:LinearColor=<rgba 0.4167 0 1.25 1>
    R2 Saffron:LinearColor=<rgba 1.25 0.9583 0 1>
    R2 SkyBlue:LinearColor=<rgba 0 0.7985 1.2 1>
    R2 TitaniumWhite:LinearColor=<rgba 1 1 1 1>
    Use R1 Painted Overrides=true
    Use R2 Painted Overrides=true
   )
   /SetVariables(Particles.VisibilityTag=&SpawnGroup()) {isSetParameters=true}
   &InitializeParticle(
    Color Mode=::NewEnumerator0
    Lifetime Max=&SelectFloatFromArray(
     Direct Array Index=&SpawnGroup()
     Float Selection Array=DataInterface() {propertyValues {FloatData [0.3,0.125]}}
    )
    Lifetime Min:NiagaraFloat=&SelectFloatFromArray(Direct Array Index=&SpawnGroup(), Float Selection Array=DataInterface() {propertyValues {FloatData [0.2,0.1]}})
    Lifetime Mode=::NewEnumerator1
    Mesh Scale Max=&SelectVectorFromArray(
     Direct Array Index=&SpawnGroup()
     Vector Selection Array=DataInterface() {propertyValues {FloatData [<2 2.5 2.5>, <2 1 1>]}}
    )
    Mesh Scale Min=&SelectVectorFromArray(
     Direct Array Index=&SpawnGroup()
     Vector Selection Array=DataInterface() {propertyValues {FloatData [<1.25 1 1>, <1 1 1>]}}
    )
    Mesh Scale Mode=::NewEnumerator2
    Sprite Rotation Mode=::NewEnumerator2
    Sprite Size Mode=::NewEnumerator3
    UsePositionOffset=false
    Write Material Random=true
   )
   &PartitionParticles(Number Of Partitions=$Emitter.PortCount)
   &InitialMeshOrientation(
    Mesh Orientation Mode=::NewEnumerator2
    Rotation=&RandomRangeVector(
     Minimum=&SelectVectorFromArray(
      Direct Array Index=&SpawnGroup()
      Vector Selection Array=DataInterface() {propertyValues {FloatData [<-1 -0.25 -0.25>, <-1 0.1 0.1>]}}
     )
     Maximum=&SelectVectorFromArray(
      Direct Array Index=&SpawnGroup()
      Vector Selection Array=DataInterface() {propertyValues {FloatData [<1 0.25 0.25>, <1 -0.1 -0.1>]}}
     )
    )
    Use Rotation Vector=true
   )
   &SkeletalMeshLocation(
    Bone / Socket Index=$Particles.PartitionParticles.Partition
    Bone / Socket Sampling Mode=::NewEnumerator5
    Interpolate Mesh Sampling=false
    MaskSpawnGroup=false
    Mesh Sampling Type=::NewEnumerator6
    Sampled Position Offset=&MaskVectorBySpawnGroup(Masked Vector=<-50 0 0>, Passthrough Vector=<-60 0 0>, Spawn Group Mask=1)
    Skeletal Mesh=$System.SkeletalMesh
   )
   &AddVelocity(Velocity=<-1500 0 0>) {enabled=false}
   &VC_ExtraBindingColors(
    ExtraColors=2
    R1 Color:LinearColor=&MakeCustomLinearColorFromBool(Bool=$User.bIsPainted, True Color=$User.PaintedColor, False Color=<rgba 16 2 16 1>)
    R2 Color:LinearColor=&MakeCustomLinearColorFromBool(Bool=$User.bIsPainted, True Color=$User.PaintedColor, False Color=<rgba 5 0.625 5 1>)
   ) {enabled=false, module=VC_ExtraBindingColors001}
  ]
  particleUpdateScript [
   &ParticleState()
   &ScaleMeshSize(Scale Factor=&MaskVector(
     Mask Source=0
     Mask Parameter=?
     Mask=1
     Masked Vector=&VectorFromCurve(VectorCurve=DataInterface() {propertyValues {
        XCurve {keys [<at 0 0>,<at 0.25 1>,<at 0.8 1>,<at 1 0.5>]}
        YCurve {keys [<at 0 0>,<at 0.25 1>,<at 0.8 1>,<at 1 0.5>]}
        ZCurve {keys [<at 0 0>,<at 0.25 1>,<at 0.8 1>,<at 1 0.5>]}
       }})
     Passthrough Vector=<1 1 1>
    ))
   &SolveForcesAndVelocity(Clamp Velocity=false, Limit Acceleration=false)
   &JitterPosition(Jitter Amount:NiagaraFloat=&MaskFloat(
     Mask Source=0
     Mask Parameter=?
     Mask=1
     Masked Float=&RandomRangeFloat(Minimum=1.25, Maximum=3)
     Passthrough Float=0
    ))
  ]
  rendererClasses [&NiagaraMeshRendererProperties, &NiagaraSpriteRendererProperties]
  renderers [
   {rendererClass=&NiagaraMeshRendererProperties, rendererIndex=0}
   {rendererClass=&NiagaraMeshRendererProperties, rendererIndex=1}
   {rendererClass=&NiagaraSpriteRendererProperties, rendererIndex=2}
  ]
 }
 NE_VC_Boost_Sprite_Parent002 {
  emitterUpdateScript [
   &EmitterState(
    Inactive Response=::NewEnumerator1
    Life Cycle Mode=::NewEnumerator1
    Loop Behavior=::NewEnumerator1
    Loop Duration Mode=::NewEnumerator1
    UseLoopDelay=false
   )
   &SpawnBurst_Instantaneous(Spawn Count=2, Use Spawn Probability=false, UseLoopCountLimit=false)
   &VC_FilteredSocketCount()
   &SpawnRate(SpawnRate=50, Use Spawn Probability=false) {enabled=false}
  ]
  particleSpawnScript [
   &VC_ExtraBindingColors(
    ExtraColors=1
    R1 Color:LinearColor=&MakeCustomLinearColorFromBool(Bool=$User.bIsPainted, True Color=$User.PaintedColor, False Color=<rgba 1.75 0.4375 1.75 1>)
   )
   &VC_PaintedVariantsOverride(
    R1 Black:LinearColor=<rgba 0.15 0.15 0.15 1>
    R1 BurntSienna:LinearColor=<rgba 1.75 0.5403 0.2157 1>
    R1 Cobalt:LinearColor=<rgba 0.2 0.5833 2.5 1>
    R1 Crimson:LinearColor=<rgba 1 0 0.0271 1>
    R1 DefaultColor=<rgba 1.75 0.4375 1.75 1>
    R1 ForestGreen:LinearColor=<rgba 0 0.8 0.0404 1>
    R1 Lime:LinearColor=<rgba 0.7583 1.3 0 1>
    R1 Orange:LinearColor=<rgba 1.75 0.525 0 1>
    R1 Purple:LinearColor=<rgba 0.4167 0 1.25 1>
    R1 Saffron:LinearColor=<rgba 1.25 0.9583 0 1>
    R1 SkyBlue:LinearColor=<rgba 0 0.7985 1.2 1>
    R1 TitaniumWhite:LinearColor=<rgba 1 1 1 1>
    Use R1 Painted Overrides=true
   )
   &InitializeParticle(
    Sprite Rotation Mode=::NewEnumerator2
    Sprite Size Mode=::NewEnumerator3
    Uniform Sprite Size=100
    UsePositionOffset=false
    Write Material Random=true
   )
   &PartitionParticles(Number Of Partitions=$Emitter.PortCount)
   &InitialMeshOrientation(
    Mesh Orientation Mode=::NewEnumerator2
    Rotation=&RandomRangeVector(Minimum=<-0.35 0 0>, Maximum=<0.35 0 0>)
    Use Rotation Vector=true
   )
   &SkeletalMeshLocation(
    Bone / Socket Index=$Particles.PartitionParticles.Partition
    Bone / Socket Sampling Mode=::NewEnumerator5
    Interpolate Mesh Sampling=false
    MaskSpawnGroup=false
    Mesh Sampling Type=::NewEnumerator6
    Sampled Position Offset=<-40 0 0>
    Skeletal Mesh=$System.SkeletalMesh
   )
   &AddVelocity(Velocity=<-1500 0 0>) {enabled=false}
  ]
  particleUpdateScript [
   &ParticleState(Kill Particles When Lifetime Has Elapsed=false, Let Infinitely Lived Particles Die When Emitter Deactivates=true)
   &SolveForcesAndVelocity(Clamp Velocity=false, Limit Acceleration=false)
  ]
  rendererClasses [&NiagaraMeshRendererProperties, &NiagaraSpriteRendererProperties]
  renderers [
   {rendererClass=&NiagaraMeshRendererProperties, rendererIndex=0}
   {rendererClass=&NiagaraSpriteRendererProperties, rendererIndex=1}
  ]
 }
}
summary {
 systemName=NS_Boost_DarkMatter
 emitters [
  {bEnabled=false, emitterName=NE_VC_Boost_Ribbon_Parent, rendererClasses [&NiagaraRibbonRendererProperties]}
  {emitterName=NE_VC_Boost_Sprite_Parent, rendererClasses [&NiagaraSpriteRendererProperties]}
  {
   emitterName=NE_VC_Boost_Sprite_Parent001
   rendererClasses [&NiagaraMeshRendererProperties, &NiagaraSpriteRendererProperties]
  }
  {
   emitterName=NE_VC_Boost_Sprite_Parent002
   rendererClasses [&NiagaraMeshRendererProperties, &NiagaraSpriteRendererProperties]
  }
  {emitterName=NE_VC_Boost_Ribbon_Parent001, rendererClasses [&NiagaraRibbonRendererProperties]}
  {emitterName=NE_VC_Boost_Ribbon_Parent002, rendererClasses [&NiagaraRibbonRendererProperties]}
 ]
 userVariables [
  User.bIsFrontendPreview {type=NiagaraBool, defaultValue=true}
  User.bIsPainted {type=NiagaraBool}
  User.ColorVariantID {type=NiagaraInt32}
  User.DTCarOffset {type=Vector3f, description="", defaultValue {struct=Vector3f, value=<0 0 0>}}
  User.DTCarOffset001 {type=NiagaraPosition, description="", defaultValue {}}
  User.DTCarOffset002 {type=NiagaraPosition, description="", defaultValue {}}
  User.GameModeID {type=NiagaraInt32}
  User.PaintedColor {type=LinearColor, description="", defaultValue {struct=LinearColor, value=<rgba 1 1 1 1>}}
  User.PaintedScalar {type=NiagaraFloat, defaultValue=1, description=""}
  User.PortCount {type=NiagaraInt32, defaultValue=2}
  User.SingleBoostPos {type=Vector3f, description="", defaultValue {struct=Vector3f, value=<0 0 0>}}
 ]
}
